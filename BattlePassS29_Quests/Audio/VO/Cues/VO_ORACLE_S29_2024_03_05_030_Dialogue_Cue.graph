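# Sound cue for one Oracle voice-over dialogue line in the Season 29 Battle Pass quests.
# Its only node is a dialogue player that plays the matching dialogue wave with the AudioLogVO_DV voice as speaker and no targets.
# It routes through the narrator VO sound class, uses the audio log attenuation settings, and has a boosted volume multiplier.
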SoundCue /BattlePassS29_Quests/Audio/VO/Cues/VO_ORACLE_S29_2024_03_05_030_Dialogue_Cue
root=DialoguePlayer_0
nodes [{
  id=DialoguePlayer_0
  properties {dialogueWaveParameter {
    dialogueWave=/BattlePassS29_Quests/Audio/VO/Waves/VO_ORACLE_S29_2024_03_05_030_Dialogue
    context {speaker=/BattlePassS29_Quests/Audio/VO/AudioLogVO_DV, targets []}
   }}
 }]
properties {
 attenuationSettings=/BattlePassS29_Quests/Audio/VO/Fort_AudioLogs_Att
 maxDistance=2097152
 soundClassObject=/Game/Sounds/FortSoundClassesAndModes/Fort_VO_Narrator
 volumeMultiplier=2
}
